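; Gameplay cue notify for the Polite Dinner consumable that triggers its progress TA playback when executed.
; On execute, it casts the target to a Fortnite player pawn and reads that pawn's current weapon.
; If the weapon is the Polite Dinner consumable, it calls PlayProgressTA on it, leaving the playback itself to that Blueprint.
; The handler returns the result of the parent OnExecute.
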(fn OnExecute
  (params (MyTarget :type Actor%20Object%20Reference) (Parameters :type Gameplay%20Cue%20Parameters%20Structure%20%28by%20ref%29))
  (bind _asfort_player_pawn_athena (Utilities|Casting|CastToFortPlayerPawnAthena MyTarget)
    (:then
      (bind _asb_consumable_polite_dinner (Utilities|Casting|CastToB_Consumable_PoliteDinner (Variables|/Script/FortniteGame.FortPawn|GetCurrentWeapon _asfort_player_pawn_athena))
        (:then
          (Class|BConsumablePoliteDinner|PlayProgressTA _asb_consumable_polite_dinner))
        (:CastFailed
          _)))
    (:CastFailed
      _))
  (return (|CallParentFunction MyTarget Parameters)))
